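; SpyTech ability AoE volume attached to a Shield Bubble+, identified in debug text as a low gravity volume, that lives as long as the shield core.
; BeginPlay sets its Sphere as the collision component before calling the parent.
; UpdateShieldRefs stores the Silver Blazer core and bubble references and binds to the core's OnDied, which calls Kill on the volume.
; Debug prints mark each stage; the gravity effect itself is not defined here.
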(event EventBeginPlay
  ;[1 1 1] Begin Play
  (Collision|SetCollisionComponent (Variables|Default|GetSphere))
  (|CallParentFunction)
  (Development|PrintString "Begin" true true "(R=1.000000,G=0.000000,B=0.661963,A=1.000000)" 5.0))

(event Custom|UpdateShieldRefs
  (params (ShieldCore :type BGA%20Athena%20Silver%20Blazer%20Core%20Object%20Reference) (ShieldBubble :type B%20BGA%20Athena%20Silver%20Blazer%20V2%20Object%20Reference))
  ;[1 1 1] Set Shield Variable references and bind to destroyed
  (Variables|Default|SetShieldCore ShieldCore)
  (Variables|Default|SetShieldBubble ShieldBubble)
  (EventDispatchers|AssignOnDied (Variables|Default|GetShieldCore) (delegate Custom|ShieldCoreDied))
  (Development|PrintString (Utilities|String|Append "Shield Core: " (Utilities|GetDisplayName (Variables|Default|GetShieldCore)) " - - - " "Shield Bubble: " (Utilities|GetDisplayName (Variables|Default|GetShieldBubble))) true true "(R=0.187575,G=1.000000,B=0.000000,A=1.000000)" 5.0))

(event Custom|ShieldCoreDied
  (params (DamagedActor :type Actor%20Object%20Reference) (Damage :type Float%20%28single-precision%29) (InstigatedBy :type Controller%20Object%20Reference) (DamageCauser :type Actor%20Object%20Reference) (HitLocation :type Vector) (FHitComponent :type Primitive%20Component%20Object%20Reference) (BoneName :type Name) (Momentum :type Vector))
  ;[1 1 1] When Shield Core is desotryed, clean up self
  (Development|PrintString "Kill attached low grav volume" true true "(R=1.000000,G=0.225168,B=0.000000,A=1.000000)" 5.0)
  (Delivery|Kill))

(fn ConstructionScript
  (|CallParentFunction))
