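; SpyTech gameplay ability that gives the player pawn a double jump via an adjustable Z launch, then resets falling height.
; Activation is blocked while skydiving or parachuting.
; It ends on landing, surface swimming, a change to walking, or an emergency timer that is cleared when the ability ends.
; When the pawn cast succeeds, the activation check ignores the parent's boolean result.
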(event Ability|EventActivateAbilityFromEvent
  (params (EventData :type Gameplay%20Event%20Data%20Structure%20%28by%20ref%29))
  (|CallParentFunction EventData)
  (bind _playerpawn (Utilities|GetValidAbilityPlayerCharacter)
    (:Valid
      (bind _output_get (Variables|SpyTech|SetPlayerPawn _playerpawn))
      (Pawn|Character|LaunchCharacterJump _output_get (Variables|SpyTech|GetJumpZAdjust) false true false true)
      ;[0.8463 1 0.4316] FORT-256725 v2
      (Pawn|Character|ResetFallingHeight (Variables|SpyTech|GetPlayerPawn))
      ;[0.44 0.9101 1] FORT-283188
      (bind _returnvalue (Utilities|Time|SetTimerbyEvent (delegate Custom|EmergyTimerTrigger) (Variables|EmergencyTimer|GetEmergencyTime)))
      (Variables|EmergencyTimer|SetEmergencyTimerHandle _returnvalue)
      ;[1 0.4107 0.4906] FORT-261259
      (EventDispatchers|AssignOnPawnLanded (Variables|SpyTech|GetPlayerPawn) (delegate Custom|OnPawnLanded))
      (Utilities|FlowControl|Sequence
        (:then_0
          (Ability|Tasks|WaitGameplayTagAdd "(TagName=\"Gameplay.Action.Player.SurfaceSwimming\")" (Variables|SpyTech|GetPlayerPawn)
            (:Added
              _)))
        (:then_1
          (Ability|Tasks|WaitMovementModeChange "MOVE_Walking"
            (:OnChange
              _))))
      (EventDispatchers|UnbindEventfromOnPawnLanded (Variables|SpyTech|GetPlayerPawn) (delegate Custom|OnPawnLanded))
      (Ability|EndAbility))
    (:Invalid
      (Ability|EndAbility))))

(event Custom|OnPawnLanded
  (params (Hit :type Hit%20Result%20Structure%20%28by%20ref%29))
  ;[1 0.4107 0.4906] FORT-261259
  (EventDispatchers|UnbindEventfromOnPawnLanded (Variables|SpyTech|GetPlayerPawn) (delegate Custom|OnPawnLanded))
  (Ability|EndAbility))

(event Custom|EmergyTimerTrigger
  ;[0.44 0.9101 1] FORT-283188
  ;[1 0.4107 0.4906] FORT-261259
  (EventDispatchers|UnbindEventfromOnPawnLanded (Variables|SpyTech|GetPlayerPawn) (delegate Custom|OnPawnLanded))
  (Ability|EndAbility))

(event Ability|EventOnEndAbility
  (params (bWasCancelled :type Boolean))
  (Utilities|Time|ClearandInvalidateTimerbyHandle (Variables|EmergencyTimer|GetEmergencyTimerHandle)))

(fn CanActivateAbility
  (params (ActorInfo :type Gameplay%20Ability%20Actor%20Info%20Structure) (Handle :type Gameplay%20Ability%20Spec%20Handle%20Structure))
  (bind (_owneractor _avataractor) (Utilities|Struct|BreakGameplayAbilityActorInfo ActorInfo))
  (bind (_relevanttags _returnvalue) (|CallParentFunction ActorInfo))
  (bind _asfort_player_pawn_athena (Utilities|Casting|CastToFortPlayerPawnAthena _avataractor)
    (:then
      (return _relevanttags (and (not (Parachute|IsSkydiving _asfort_player_pawn_athena)) (not (Parachute|IsParachuteOpen _asfort_player_pawn_athena)))))
    (:CastFailed
      (return _relevanttags _returnvalue))))
