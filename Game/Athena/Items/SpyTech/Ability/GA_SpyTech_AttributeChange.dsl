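; GA_SpyTech_AttributeChange is a gameplay ability for SpyTech items that applies a configurable gameplay effect to its owner.
; On activation it reads EffectToApply and checks for a valid ability player character.
; If valid, it prints the effect's class name as debug output, applies the effect and ends.
; If invalid, it ends without applying anything.
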(event Ability|EventActivateAbility
  (bind _effecttoapply (Variables|Default|GetEffectToApply))
  (Utilities|GetValidAbilityPlayerCharacter
    (:Valid
      (Development|PrintString (Utilities|GetClassDisplayName _effecttoapply))
      (Ability|ApplyGameplayEffectToOwner _effecttoapply 0)
      (Ability|EndAbility))
    (:Invalid
      (Ability|EndAbility))))
