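; Gameplay ability for the SpyTech Chug Splash+ bonus heal, applying a shield-over-time effect when triggered by a gameplay event.
; On activation it breaks the event data and applies the GE_ShieldHoT effect class variable to the event's target data.
; It then ends immediately, leaving any ongoing restoration to the gameplay effect.
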(event Ability|EventActivateAbilityFromEvent
  (params (EventData :type Gameplay%20Event%20Data%20Structure%20%28by%20ref%29))
  (bind (_eventtag _instigator _target _optionalobject _optionalobject2 _contexthandle _instigatortags _targettags _eventmagnitude _targetdata) (Utilities|Struct|BreakGameplayEventData EventData))
  (Ability|ApplyGameplayEffectToTarget _targetdata (Variables|Default|GetGE_ShieldHoT))
  (Ability|EndAbility))
