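; Passive SpyTech gameplay ability "Come At Me Bro" that exposes its owning player to all non-friendly players in Athena matches.
; On activation it calls the parent, binds to the game state's players-left delegate and runs the handler once immediately.
; Each run loops the player array and adds the owner's pawn to non-friendly controllers' indicated and stenciled lists indefinitely.
; Particle, sound, stencil, sharing and screen-clamp settings come from variables.
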(event Ability|EventActivateAbility
  (bind _returnvalue (Fortnite|Athena|GameState|GetGameStateAthena))
  ;[1 1 1] On activate ability (when granted since passive)
  (|CallParentFunction)
  (Game|AssignOnPlayersLeftChanged _returnvalue (EventDispatchers|CreateEvent :!event Custom|PlayersLeftChanged))
  (CallFunction|PlayersLeftChanged (Variables|/Script/FortniteGame.FortGameStateAthena|GetPlayersLeft _returnvalue)))

(event Custom|PlayersLeftChanged
  (params (NumPlayersLeft :type Integer))
  (bind _arr_playerpawn (Variables|Default|GetArr_PlayerPawn))
  ;[1 1 1] On player count changed, iterate through players and add self to non-friendly players' indicated/stencil lists
  (bind _playerpawn (Utilities|GetValidAbilityPlayerCharacter)
    (:Valid
      (bind _output_get (Variables|Default|SetPlayerPawn _playerpawn))
      (Variables|Default|SetArr_PlayerPawn (Utilities|Array|MakeArray _output_get))
      (for _array_element (Variables|/Script/Engine.GameStateBase|GetPlayerArray (Fortnite|Athena|GameState|GetGameStateAthena))
        (if (Utilities|Enum|NotEqual(Enum) (Team|GetTeamAffiliation (Fortnite|GetFortPlayerControllerfromActor _array_element) (Variables|Default|GetPlayerPawn)) "Friendly")
          (bind _asfort_player_controller_athena (Utilities|Casting|CastToFortPlayerControllerAthena (Fortnite|GetFortPlayerControllerfromActor _array_element)))
          (IndicatedManagement|AddActorsToIndicatedList _asfort_player_controller_athena _arr_playerpawn (IndicatedManagement|MakeIndicatedActorData :GroupIdentifier (Variables|Default|GetIndicatedGroup) :ParticleSystemData (IndicatedManagement|MakeIndicatedActorParticleSystemData (Variables|Default|GetIndicatedParticleSystem) (Variables|Default|GetIndicatedParticleSystemParamName) (Variables|Default|GetPS_Offset) (Variables|Default|GetPS_DBNO_Offset) (Variables|Default|GetPS_VectorParamName)) :EventTags (Tags|GetAbilityTags) :Duration -1.0 :StepTime 0.1 :IndicatorOffset "0.000000,0.000000,90.000000" :IndicatorDBNOOffset "0.000000,0.000000,30.000000" :ShareActorWith (Variables|Default|GetShareActorWith) :Sound (Variables|Default|GetIndicatedSound) :StateImageOverride "None" :StateIndexOverride -1 :bClampToScreen (> (ScalableFloat|GetValueatLevel (Variables|Default|GetB_ClampToScreen)) 0.0) :bHideStateImage false :bOnlyDisplayWhenStencilingIsDisabled false :bOnlyDisplayIfActorIsVisible false) (Variables|Default|GetAddAsUnique))
          (IndicatedManagement|AddActorsToStenciledList _asfort_player_controller_athena _arr_playerpawn (IndicatedManagement|MakeStenciledActorData :GroupIdentifier (Variables|Default|GetIndicatedGroup) :Duration -1.0 :StepTime 0.1 :ShareActorWith (Variables|Default|GetShareActorWith) :FriendlyStencilName "None" :EnemyStencilName (Variables|Default|GetEnemyStencilName) :bOnlyDisplayIfActorIsVisible false) (Variables|Default|GetAddAsUnique)))))))
