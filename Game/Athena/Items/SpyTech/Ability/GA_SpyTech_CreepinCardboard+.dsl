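; SpyTech gameplay ability that overrides the max health of a Creepin' Cardboard prop, an Athena_Player_CoolCarpet actor built on SneakySnowmanV2.
; An activation event supplies the prop as its optional object.
; Unless both the actor's health and its MaxHealth setting already match the override, it writes MaxHealthOverride, unsuppresses the health bar and applies GE_Health through the player's ability system.
; Debug strings trace each outcome, and the ability ends after a successful cast.
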(event Ability|EventActivateAbilityFromEvent
  (params (EventData :type Gameplay%20Event%20Data%20Structure%20%28by%20ref%29))
  (bind (_eventtag _instigator _target _optionalobject) (Utilities|Struct|BreakGameplayEventData EventData))
  (|CallParentFunction EventData)
  (Development|PrintString "Try Set Max Health" true true "(R=1.000000,G=0.712717,B=0.000000,A=1.000000)" 5.0)
  (bind _asathena_player_cool_carpet (Utilities|Casting|CastToAthena_Player_CoolCarpet _optionalobject)
    (:then
      (Variables|Default|SetCreepinCardboard _asathena_player_cool_carpet)
      (if (not (and (== (Health|GetMaxHealth (Variables|Default|GetCreepinCardboard) :!class /Script/FortniteGame.BuildingActor) (ScalableFloat|GetValueatLevel (Variables|/Game/Athena/Items/Consumables/SneakySnowman/Version2/Athena_Player_SneakySnowmanV2.Athena_Player_SneakySnowmanV2_C|GetMaxHealth (Variables|Default|GetCreepinCardboard)))) (== (ScalableFloat|GetValueatLevel (Variables|/Game/Athena/Items/Consumables/SneakySnowman/Version2/Athena_Player_SneakySnowmanV2.Athena_Player_SneakySnowmanV2_C|GetMaxHealth (Variables|Default|GetCreepinCardboard))) (ScalableFloat|GetValueatLevel (Variables|Default|GetMaxHealthOverride)))))
        (bind _playerpawn (Utilities|GetValidAbilityPlayerCharacter)
          (:Valid
            (Variables|Default|SetPlayer _playerpawn)
            (Development|PrintString "Max Health Set" true true "(R=0.099351,G=1.000000,B=0.000000,A=1.000000)" 5.0)
            (Class|BuildingActor|SetSuppressHealthBar (Variables|Default|GetCreepinCardboard) false)
            (Variables|/Game/Athena/Items/Consumables/SneakySnowman/Version2/Athena_Player_SneakySnowmanV2.Athena_Player_SneakySnowmanV2_C|SetMaxHealth (Variables|Default|GetMaxHealthOverride) (Variables|Default|GetCreepinCardboard))
            (Fortnite|ApplyGameplayEffectsToActor (Variables|Default|GetCreepinCardboard) (Variables|Default|GetGE_Health) (Transformation|GetActorLocation (Variables|Default|GetPlayer)) false 0 (Ability|GetAbilitySystemComponent (Variables|Default|GetPlayer)))
            (Ability|EndAbility))
          (:Invalid
            _))
        (else
          (Development|PrintString "Max Health Already Set" true true "(R=1.000000,G=0.000000,B=0.014146,A=1.000000)" 5.0)))
      (Ability|EndAbility))
    (:CastFailed
      _)))
