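; Gameplay ability for the SpyTech Decoy+ bonus shields item, which applies a gameplay effect to the decoy pawn when it spawns.
; On activation it calls the parent ability, then waits for the DangerGrape spawned gameplay event.
; It casts the event's optional object to a Fort pawn and applies the configured decoy effect to it, using the valid player character's location and ability system component as the source.
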(event Ability|EventActivateAbility
  (|CallParentFunction)
  (bind (_asynctaskproxy _payload) (Ability|Tasks|WaitGameplayEvent "(TagName=\"Abilities.Generic.DangerGrape.Spawned\")")
    (:EventReceived
      (bind _playerpawn (Utilities|GetValidAbilityPlayerCharacter)
        (:Valid
          (bind _asfort_pawn (Utilities|Casting|CastToFortPawn (.OptionalObject (Utilities|Struct|BreakGameplayEventData _payload))))
          (Fortnite|ApplyGameplayEffectsToActor _asfort_pawn (Variables|Default|GetGE_ApplyToDecoy) (Transformation|GetActorLocation _playerpawn) false 0 (Ability|GetAbilitySystemComponent _playerpawn)))))))
