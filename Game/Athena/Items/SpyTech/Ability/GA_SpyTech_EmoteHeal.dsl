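; SpyTech emote heal gameplay ability that periodically heals nearby players while it is active.
; On event activation, it starts a looping timer at a scalable-float period, so players entering or leaving the area are picked up.
; Each tick runs target selection and applies the gameplay effect container to the resulting targets.
; Ability end clears the timer; two helper functions resolve the avatar's player pawn and controller.
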(event Ability|EventActivateAbilityFromEvent
  (params (EventData :type Gameplay%20Event%20Data%20Structure%20%28by%20ref%29))
  (|CallParentFunction EventData)
  (bind _returnvalue (Utilities|Time|SetTimerbyEvent (delegate Custom|TriggerHeal) (ScalableFloat|GetValueatLevel (Variables|Default|GetSF_HealPeriod)) true)) ; Set a timer for our period since players can enter/exit the area
  (Variables|Default|SetHandle_HealTimer _returnvalue))

(event Custom|TriggerHeal
  (bind (_asynctaskproxy _targetdata _applicationtag) (Ability|Tasks|DoTargetSelection "None" "(TagName=\"Targeting.ApplicationTag.0\")" true)
    (:Targeted
      (Ability|ApplyGameplayEffectContainer _targetdata _applicationtag))) ; Check for nearby players to heal
)

(event Ability|EventOnEndAbility
  (params (bWasCancelled :type Boolean))
  ;[1 1 1] Clean up timer on ability end
  (Utilities|Time|ClearandInvalidateTimerbyHandle (Variables|Default|GetHandle_HealTimer))
  (|CallParentFunction bWasCancelled))

(fn GetAbilityPlayerController
  (bind _returnvalue (Fortnite|GetFortPlayerControllerfromActor (.AsFort%20Player%20Pawn%20Athena (CallFunction|GetAbilityPlayerPawn))))
  (bind (_valid _asfort_player_pawn_athena) (CallFunction|GetAbilityPlayerPawn))
  (if _valid
    (Utilities|IsValid _returnvalue
      (:Is%20Valid
        (return true _returnvalue))
      (:Is%20Not%20Valid
        _))
    (else
      _))
  (return false 0))

(fn GetAbilityPlayerPawn
  (bind _asfort_player_pawn_athena (Utilities|Casting|CastToFortPlayerPawnAthena (Ability|GetAvatarActorfromActorInfo))
    (:then
      (return true _asfort_player_pawn_athena))
    (:CastFailed
      (return false 0))))
