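; SpyTech gameplay ability for the Bandage Bazooka that refills the item's charge on the player pawn.
; It calls the parent logic, caches the avatar pawn, then sets up one-shot-gated listeners for an end-of-round event and the charge attribute reaching a threshold, the latter followed by a delay.
; Refill Ammo applies gameplay effects setting max and current charge, and the ability ends itself so it can be re-triggered.
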(event Ability|EventActivateAbility
  ;[0.03989 1 0] Activate Ability
  (|CallParentFunction)
  (bind _asfort_player_pawn_athena (Utilities|Casting|CastToFortPlayerPawnAthena (Ability|GetAvatarActorfromActorInfo)))
  (Variables|Default|SetMyPawn _asfort_player_pawn_athena)
  (Utilities|FlowControl|DoOnce false
    (:Completed
      (Utilities|FlowControl|Sequence
        (:then_0
          ;;[0 1 0.7834] Set-up listeners for events that will refill ammo
          (Utilities|FlowControl|DoOnce false
            (:Completed
              (Ability|Tasks|WaitGameplayEvent (Variables|Default|GetEndOfRoundTag%20Const)
                (:then
                  (Ability|Tasks|WaitForAttributeChangeWithComparison :InAttribute (Variables|Default|GetBandage%20Bazooka%20Charge%20Atribute%20Const) :InComparisonType "GreaterThanOrEqualTo" :InComparisonValue 1.0 :TriggerOnce true
                    (:OnChange
                      (Utilities|FlowControl|Delay (Variables|Default|GetHeal%20Charge%20Refill%20Ammo%20Delay%20Const)))))
                (:EventReceived
                  _))
              (CallFunction|RefillAmmo)
              ;;[1 0.456 0.2396] End Ability
              ;; This allows the ability to be re-triggered later
              (Ability|EndAbility))))
        (:then_1
          _)))))

(event Ability|EventOnEndAbility
  (params (bWasCancelled :type Boolean))
  ;[1 0.4667 0] End Ability
  (|CallParentFunction bWasCancelled)
  ;[0.03989 1 0] Activate Ability
  (Utilities|FlowControl|DoOnce false
    (:Completed
      (Utilities|FlowControl|Sequence
        (:then_0
          ;;[0 1 0.7834] Set-up listeners for events that will refill ammo
          (Utilities|FlowControl|DoOnce false
            (:Completed
              (Ability|Tasks|WaitGameplayEvent (Variables|Default|GetEndOfRoundTag%20Const)
                (:then
                  (Ability|Tasks|WaitForAttributeChangeWithComparison :InAttribute (Variables|Default|GetBandage%20Bazooka%20Charge%20Atribute%20Const) :InComparisonType "GreaterThanOrEqualTo" :InComparisonValue 1.0 :TriggerOnce true
                    (:OnChange
                      (Utilities|FlowControl|Delay (Variables|Default|GetHeal%20Charge%20Refill%20Ammo%20Delay%20Const)))))
                (:EventReceived
                  _))
              (CallFunction|RefillAmmo)
              ;;[1 0.456 0.2396] End Ability
              ;; This allows the ability to be re-triggered later
              (Ability|EndAbility))))
        (:then_1
          _)))))

(event Custom|Refill%20Ammo
  ;[0.1568 0.6178 1] Refill Ammo
  (bind _mypawn (Variables|Default|GetMyPawn)
    (:then
      (Ability|ApplyGameplayEffectToTarget (Ability|TargetData|AbilityTargetDatafromActor _mypawn) (Variables|Default|GetBandageBazookaSetMaxCharge) 0)
      (Ability|ApplyGameplayEffectToTarget (Ability|TargetData|AbilityTargetDatafromActor _mypawn) (Variables|Default|GetBadnageBazookaSetCurrentCharge) 0))
    (:else)))
