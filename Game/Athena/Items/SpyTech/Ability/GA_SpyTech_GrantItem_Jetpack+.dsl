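; SpyTech ability, apparently a child of an item-granting ability, that applies a jetpack fuel gameplay effect to the player pawn.
; After parent activation, a once-only path checks the max fuel attribute, waiting for it to become positive plus a short delay if needed, before applying the effect.
; A gameplay event listener reapplies it on each new LTM round.
; Ability end repeats the same once-only fuel logic using the pawn cached at activation.
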(event Custom|WaitForRoundReset
  (Ability|Tasks|WaitGameplayEvent "(TagName=\"Event.LTM.OnNewRound\")"
    (:EventReceived
      (Ability|ApplyGameplayEffectToTarget (Ability|TargetData|AbilityTargetDatafromActor (stale Ability|EventActivateAbility _asfort_player_pawn_athena)) (Variables|Default|GetGE_SetJetpackFuel) 0))))

(event Ability|EventActivateAbility
  (|CallParentFunction)
  (CallFunction|WaitforRoundReset)
  (bind _asfort_player_pawn_athena (Utilities|Casting|CastToFortPlayerPawnAthena (Ability|GetAvatarActorfromActorInfo)))
  (Utilities|FlowControl|DoOnce false
    (:Completed
      (if (not (and (Ability|Attribute|GetFloatAttributefromAbilitySystemComponent (Ability|GetFortAbilitySystemComponent _asfort_player_pawn_athena) (Variables|Default|GetAttr_MaxFuel)) (> (.ReturnValue (Ability|Attribute|GetFloatAttributefromAbilitySystemComponent (Ability|GetFortAbilitySystemComponent _asfort_player_pawn_athena) (Variables|Default|GetAttr_MaxFuel))) 0.0)))
        (Ability|Tasks|WaitForAttributeChangeWithComparison :InAttribute (Variables|Default|GetAttr_MaxFuel) :InComparisonType "GreaterThanOrEqualTo" :InComparisonValue 1.0 :TriggerOnce true
          (:OnChange
            (Utilities|FlowControl|Delay 2.0))))
      (Ability|ApplyGameplayEffectToTarget (Ability|TargetData|AbilityTargetDatafromActor _asfort_player_pawn_athena) (Variables|Default|GetGE_SetJetpackFuel) 0))))

(event Ability|EventOnEndAbility
  (params (bWasCancelled :type Boolean))
  (|CallParentFunction bWasCancelled)
  (Utilities|FlowControl|DoOnce false
    (:Completed
      (if (not (and (Ability|Attribute|GetFloatAttributefromAbilitySystemComponent (Ability|GetFortAbilitySystemComponent (stale Ability|EventActivateAbility _asfort_player_pawn_athena)) (Variables|Default|GetAttr_MaxFuel)) (> (.ReturnValue (Ability|Attribute|GetFloatAttributefromAbilitySystemComponent (Ability|GetFortAbilitySystemComponent (stale Ability|EventActivateAbility _asfort_player_pawn_athena)) (Variables|Default|GetAttr_MaxFuel))) 0.0)))
        (Ability|Tasks|WaitForAttributeChangeWithComparison :InAttribute (Variables|Default|GetAttr_MaxFuel) :InComparisonType "GreaterThanOrEqualTo" :InComparisonValue 1.0 :TriggerOnce true
          (:OnChange
            (Utilities|FlowControl|Delay 2.0))))
      (Ability|ApplyGameplayEffectToTarget (Ability|TargetData|AbilityTargetDatafromActor (stale Ability|EventActivateAbility _asfort_player_pawn_athena)) (Variables|Default|GetGE_SetJetpackFuel) 0))))
