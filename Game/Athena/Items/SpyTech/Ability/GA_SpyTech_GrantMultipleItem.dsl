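; GA_SpyTech_GrantMultipleItem is a SpyTech gameplay ability Blueprint, named for granting multiple items, though no granting logic appears here.
; Its activation event has no steps attached.
; A helper, GetAbilityPlayerPawn, casts the avatar actor to FortPlayerPawnAthena and returns a validity flag with the pawn, or false and null when the cast fails.
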(event Ability|EventActivateAbility)

(fn GetAbilityPlayerPawn
  (bind _asfort_player_pawn_athena (Utilities|Casting|CastToFortPlayerPawnAthena (Ability|GetAvatarActorfromActorInfo))
    (:then
      (return (Utilities|IsValid _asfort_player_pawn_athena :!class /Script/Engine.KismetSystemLibrary) _asfort_player_pawn_athena))
    (:CastFailed
      (return false 0))))
