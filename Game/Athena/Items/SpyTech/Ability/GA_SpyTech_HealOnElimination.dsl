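; SpyTech gameplay ability that heals its owner when triggered by an elimination-related gameplay event.
; The response check accepts any earned elimination event; other events qualify only if the target carries all required NPC tags.
; On activation, it builds a heal effect spec, sets the healing magnitude via SetByCaller from a scalable float, applies it to the owner, then ends.
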(event Ability|EventActivateAbilityFromEvent
  (params (EventData :type Gameplay%20Event%20Data%20Structure%20%28by%20ref%29))
  (|CallParentFunction EventData)
  (bind _returnvalue (Ability|GameplayEffect|AssignTagSetbyCallerMagnitude (Ability|MakeOutgoingGameplayEffectSpec (Variables|Default|GetGE_Heal)) "(TagName=\"SetByCaller.Healing\")" (ScalableFloat|GetValueatLevel (Variables|Default|GetHealOnElimination))))
  (Ability|ApplyGameplayEffectSpecToOwner _returnvalue)
  (Ability|EndAbility))

(fn ShouldAbilityRespondToEvent
  (params (ActorInfo :type Gameplay%20Ability%20Actor%20Info%20Structure) (Payload :type Gameplay%20Event%20Data%20Structure))
  (bind (_eventtag _instigator _target _optionalobject _optionalobject2 _contexthandle _instigatortags _targettags) (Utilities|Struct|BreakGameplayEventData Payload))
  (if (GameplayTags|MatchesTag _eventtag "(TagName=\"Event.EarnedElimination\")")
    (return true)
    (else
      (return (GameplayTags|HasAllTags _targettags (Variables|Default|GetRequired%20NPC%20Tags))))))
