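; GA_SpyTech_HealSteal is a spy tech gameplay ability for Infiltration that gives life steal, healing the damager.
; An activating event supplies the magnitude and instigator, whose player pawn becomes the effect target.
; It multiplies that magnitude by the HealStealPercent curve row from the Infiltration game data.
; The result feeds the HealSteal effect as a SetByCaller healing magnitude; the ability ends only on the row-found path.
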(event Ability|EventActivateAbilityFromEvent
  (params (EventData :type Gameplay%20Event%20Data%20Structure%20%28by%20ref%29))
  (bind (_eventtag _instigator _target _optionalobject _optionalobject2 _contexthandle _instigatortags _targettags _eventmagnitude) (Utilities|Struct|BreakGameplayEventData EventData))
  (bind _output_get (Variables|Default|SetDamager (Game|Player|GetPlayerPawn (Fortnite|GetFortPlayerControllerfromActor _instigator))))
  (Variables|Default|SetTarget (Ability|TargetData|AbilityTargetDatafromActor _output_get))
  (bind _outxy (DataTable|EvaluateCurveTableRow "/Game/Athena/Playlists/Infiltration/OverrideGameData.OverrideGameData" "Infiltration.HealStealPercent")
    (:RowFound
      (bind _output_get_2 (Variables|Default|SetLifeStealAmount (* _outxy _eventmagnitude)))
      (bind _returnvalue (Ability|GameplayEffect|AssignTagSetbyCallerMagnitude (Ability|MakeOutgoingGameplayEffectSpec (Variables|Default|GetGE%20HealSteal)) "(TagName=\"SetByCaller.Healing\")" _output_get_2))
      (Ability|ApplyGameplayEffectSpecToTarget _returnvalue (Variables|Default|GetTarget))
      (Ability|EndAbility))))
