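; SpyTech gameplay ability that gives its owner regeneration after taking damage.
; On activation it scales the event magnitude by the recovered percent and by effect period over duration, giving a per-tick amount that totals that percent of the damage.
; Damage.Received events set it as HealthRegen and Damage.Shielded events as ShieldRegen on a GE_Heal spec applied to the owner.
; The ability then ends, as it does for any other tag.
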(event Ability|EventActivateAbilityFromEvent
  (params (EventData :type Gameplay%20Event%20Data%20Structure%20%28by%20ref%29))
  (bind (_eventtag _instigator _target _optionalobject _optionalobject2 _contexthandle _instigatortags _targettags _eventmagnitude) (Utilities|Struct|BreakGameplayEventData EventData))
  (bind _percenthealthrecovered (Variables|Default|GetPercentHealthRecovered))
  (bind _returnvalue (ScalableFloat|GetValueatLevel _percenthealthrecovered))
  (bind _effectduration (Variables|Default|GetEffectDuration))
  (bind _returnvalue_2 (ScalableFloat|GetValueatLevel _effectduration))
  (bind _returnvalue_3 (/ 1.0 _returnvalue_2))
  (bind _effectperiod (Variables|Default|GetEffectPeriod))
  (bind _returnvalue_4 (ScalableFloat|GetValueatLevel _effectperiod))
  (bind _returnvalue_5 (Math|Float|float*float _eventmagnitude _returnvalue _returnvalue_4 _returnvalue_3))
  (|CallParentFunction EventData)
  (if (GameplayTags|Equal(GameplayTag) _eventtag "(TagName=\"Event.Damage.Received\")")
    (bind _returnvalue_6 (Ability|GameplayEffect|AssignTagSetbyCallerMagnitude (Ability|MakeOutgoingGameplayEffectSpec (Variables|Default|GetGE_Heal)) "(TagName=\"SetByCaller.HealthRegen\")" _returnvalue_5))
    (Ability|ApplyGameplayEffectSpecToOwner _returnvalue_6)
    (Ability|EndAbility)
    (elif (GameplayTags|Equal(GameplayTag) _eventtag "(TagName=\"Event.Damage.Shielded\")")
      (bind _returnvalue_7 (Ability|GameplayEffect|AssignTagSetbyCallerMagnitude (Ability|MakeOutgoingGameplayEffectSpec (Variables|Default|GetGE_Heal)) "(TagName=\"SetByCaller.ShieldRegen\")" _returnvalue_5))
      (Ability|ApplyGameplayEffectSpecToOwner _returnvalue_7)
      (Ability|EndAbility)
      (else
        (Ability|EndAbility)))))

(fn GetAbilityPlayerController
  (bind _returnvalue (Fortnite|GetFortPlayerControllerfromActor (.AsFort%20Player%20Pawn%20Athena (CallFunction|GetAbilityPlayerPawn))))
  (bind (_valid _asfort_player_pawn_athena) (CallFunction|GetAbilityPlayerPawn))
  (if _valid
    (Utilities|IsValid _returnvalue
      (:Is%20Valid
        (return true _returnvalue))
      (:Is%20Not%20Valid
        _))
    (else
      _))
  (return false 0))

(fn GetAbilityPlayerPawn
  (bind _asfort_player_pawn_athena (Utilities|Casting|CastToFortPlayerPawnAthena (Ability|GetAvatarActorfromActorInfo))
    (:then
      (return true _asfort_player_pawn_athena))
    (:CastFailed
      (return false 0))))
