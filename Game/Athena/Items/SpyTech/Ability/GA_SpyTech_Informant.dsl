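; Spy Tech Informant gameplay ability that reveals nearby actors to the player's team after an event.
; It responds to earned-elimination events, or events whose target has all required NPC tags.
; It uses the event target's location as a custom source transform for target selection, then gives found actors timed indicators and red enemy stencils shared with the team.
; Debug sphere and name printing remain, marked for deletion.
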(event Ability|EventActivateAbilityFromEvent
  (params (EventData :type Gameplay%20Event%20Data%20Structure%20%28by%20ref%29))
  (bind (_eventtag _instigator _target) (Utilities|Struct|BreakGameplayEventData EventData))
  (bind _duration (Variables|Default|GetDuration))
  (bind _returnvalue (ScalableFloat|GetValueatLevel _duration))
  (bind _returnvalue_2 (Transformation|GetActorLocation _target))
  (bind _indicator_group (Variables|Default|GetIndicator%20Group))
  (|CallParentFunction EventData)
  ;[1 0 0] To be deleted
  (Rendering|Debug|DrawDebugSphere _returnvalue_2 1e+02 12 "(R=1.000000,G=1.000000,B=1.000000,A=1.000000)" 5.0 1e+01)
  (bind _playerpawn (Utilities|GetValidAbilityPlayerCharacter)
    (:Valid
      (bind _asfort_player_controller_athena (Utilities|Casting|CastToFortPlayerControllerAthena (Fortnite|GetFortPlayerControllerfromActor _playerpawn))
        (:then
          (Variables|Default|SetPlayerController _asfort_player_controller_athena)
          (Variables|Default|SetCustomAbilitySourceTransform (Math|Transform|MakeTransform _returnvalue_2))
          (bind (_asynctaskproxy _targetdata) (Ability|Tasks|DoTargetSelection "None" "(TagName=\"Targeting.ApplicationTag.0\")")
            (:Targeted
              (if (> (Utilities|Array|Length (Ability|TargetData|GetAllActorsfromTargetData _targetdata)) 0)
                (bind _output_get (Variables|Default|SetIndicatedActors (Ability|TargetData|GetAllActorsfromTargetData _targetdata)))
                ;[1 0 0] To be deleted
                (for _array_element _output_get
                  (Development|PrintString (Utilities|GetDisplayName _array_element)))
                (IndicatedManagement|AddActorsToIndicatedList (Variables|Default|GetPlayerController) (Variables|Default|GetIndicatedActors) (IndicatedManagement|MakeIndicatedActorData :GroupIdentifier _indicator_group :ParticleSystemData (IndicatedManagement|MakeIndicatedActorParticleSystemData (Variables|Default|GetIndicated%20PS) (Variables|Default|GetParticle%20System%20Param%20Name) (Variables|Default|GetPS%20Offset) (Variables|Default|GetPS%20DBNO%20Offset) (Variables|Default|GetPS%20Vector%20Param%20Name)) :EventTags (Tags|GetAbilityTags) :Duration _returnvalue :StepTime 0.1 :IndicatorOffset "0.000000,0.000000,90.000000" :IndicatorDBNOOffset "0.000000,0.000000,30.000000" :ShareActorWith "AllTeam" :Sound (Variables|Default|GetIndicated%20Sound) :StateImageOverride "None" :StateIndexOverride -1 :bClampToScreen false :bHideStateImage false :bOnlyDisplayWhenStencilingIsDisabled false :bOnlyDisplayIfActorIsVisible false))
                (IndicatedManagement|AddActorsToStenciledList (Variables|Default|GetPlayerController) (Variables|Default|GetIndicatedActors) (IndicatedManagement|MakeStenciledActorData :GroupIdentifier _indicator_group :Duration _returnvalue :StepTime 0.1 :ShareActorWith "AllTeam" :FriendlyStencilName "None" :EnemyStencilName "RedOverlayID" :bOnlyDisplayIfActorIsVisible false))
                (else
                  _))
              (Ability|EndAbility))
            (:Cancelled
              _)))
        (:CastFailed
          _)))
    (:Invalid
      _)))

(fn GetCustomAbilitySourceTransform
  (return (Variables|Default|GetCustomAbilitySourceTransform)))

(fn ShouldAbilityRespondToEvent
  (params (ActorInfo :type Gameplay%20Ability%20Actor%20Info%20Structure) (Payload :type Gameplay%20Event%20Data%20Structure))
  (bind (_eventtag _instigator _target _optionalobject _optionalobject2 _contexthandle _instigatortags _targettags) (Utilities|Struct|BreakGameplayEventData Payload))
  (if (GameplayTags|MatchesTag _eventtag "(TagName=\"Event.EarnedElimination\")")
    (return true)
    (else
      (return (GameplayTags|HasAllTags _targettags (Variables|Default|GetRequired%20NPC%20Tags))))))
