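; Gameplay ability for the SpyTech Jetpack+ that tracks whether the player's reserve fuel is full and starts fuel regeneration.
; On activation it starts a looping half-second CheckFuel timer, then applies the FuelRegen effect to the player after a short delay.
; CheckFuel compares the ReserveFuel attribute to ReserveFuelMax, adding or removing a loose HasMaxReserveFuel tag only on state changes, with debug prints.
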(event Ability|EventActivateAbility
  (|CallParentFunction)
  (bind _returnvalue (Utilities|Time|SetTimerbyEvent (delegate Custom|CheckFuel) 0.5 true))
  (Variables|Default|SetTimerHandle _returnvalue)
  (Utilities|FlowControl|Delay 2.0)
  (bind _playerpawn (Utilities|GetValidAbilityPlayerCharacter)
    (:Valid
      (GameplayEffects|ApplyGameplayEffectToSelf :self (Ability|GetFortAbilitySystemComponent _playerpawn) :GameplayEffectClass (Variables|Default|GetGE_FuelRegen) :Level 0.0))))

(event Custom|CheckFuel
  (bind _tag_hasmaxreservefuel (Variables|Default|GetTag_HasMaxReserveFuel))
  (bind _playerpawn (Variables|Default|GetPlayerPawn))
  (bind _returnvalue (GameplayTags|HasMatchingGameplayTag _playerpawn _tag_hasmaxreservefuel))
  (bind _tag_hasmaxreservefuel_2 (Variables|Default|GetTag_HasMaxReserveFuel))
  (bind _returnvalue_2 (GameplayTags|MakeGameplayTagContainerfromTag _tag_hasmaxreservefuel_2))
  (bind (_bsuccess _returnvalue_3) (Ability|Attribute|EvaluateAttributeValuewithTags (Ability|GetFortAbilitySystemComponent (Variables|Default|GetPlayerPawn)) (Variables|Default|GetAttribute_ReserveFuel) (GameplayTags|MakeLiteralGameplayTagContainer) (GameplayTags|MakeLiteralGameplayTagContainer)))
  (bind (_bsuccess_2 _returnvalue_4) (Ability|Attribute|EvaluateAttributeValuewithTags (Ability|GetFortAbilitySystemComponent (Variables|Default|GetPlayerPawn)) (Variables|Default|GetAttribute_ReserveFuelMax) (GameplayTags|MakeLiteralGameplayTagContainer) (GameplayTags|MakeLiteralGameplayTagContainer)))
  (bind _playerpawn_2 (Utilities|GetValidAbilityPlayerCharacter)
    (:Valid
      (Variables|Default|SetPlayerPawn _playerpawn_2)
      (if (>= _returnvalue_3 _returnvalue_4)
        (if (not _returnvalue)
          (Ability|GameplayEffect|AddLooseGameplayTags (Variables|Default|GetPlayerPawn) _returnvalue_2)
          (Development|PrintString "MaxFuelAdded"))
        (elif _returnvalue
          (Ability|GameplayEffect|RemoveLooseGameplayTags (Variables|Default|GetPlayerPawn) _returnvalue_2)
          (Development|PrintString "MaxFuelRemoved"))))))
