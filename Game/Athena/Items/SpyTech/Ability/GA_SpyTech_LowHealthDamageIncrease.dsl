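; SpyTech gameplay ability that gives the owning player a damage multiplier effect while health plus shield is below a scalable threshold.
; On activation it starts repeating attribute-change tasks on current health and current shield, which drive the threshold check.
; Below the threshold it removes the stored effect handle and reapplies the effect, avoiding stacked copies.
; Otherwise it removes the effect through that handle.
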(event Ability|EventActivateAbility
  (|CallParentFunction)
  (Ability|Tasks|WaitForAttributeChange :Attribute (Variables|Default|GetAtt_CurrentHealth) :TriggerOnce false
    (:then
      (Ability|Tasks|WaitForAttributeChange :Attribute (Variables|Default|GetAtt_CurrentShield) :TriggerOnce false
        (:OnChange
          _)) ; Wait for Shield change
    )
    (:OnChange
      _)) ; Wait for HP change
  (if (CallFunction|CheckTotalPlayerHealth)
    (Ability|RemoveGameplayEffectFromOwnerWithHandle (Variables|Default|GetHandle_DamageMult)) ; If we already have the buff, remove to properly reapply.
    (bind _returnvalue (Ability|ApplyGameplayEffectToOwner (Variables|Default|GetGE_DamageMult)))
    (Variables|Default|SetHandle_DamageMult _returnvalue)
    (else
      (Ability|RemoveGameplayEffectFromOwnerWithHandle (Variables|Default|GetHandle_DamageMult)) ; Remove damage mult when our health is > threshold
    )))

(fn CheckTotalPlayerHealth
  (bind (_valid _asfort_player_pawn_athena) (CallFunction|GetAbilityPlayerPawn))
  (bind (_bsuccessfullyfoundattribute _returnvalue) (Ability|Attribute|GetFloatAttribute _asfort_player_pawn_athena (Variables|Default|GetAtt_CurrentHealth)))
  (bind (_bsuccessfullyfoundattribute_2 _returnvalue_2) (Ability|Attribute|GetFloatAttribute _asfort_player_pawn_athena (Variables|Default|GetAtt_CurrentShield)))
  (if _valid
    (return (< (+ _returnvalue _returnvalue_2) (ScalableFloat|GetValueatLevel (Variables|Default|GetSF_LowHealthThreshold))))
    (else
      (return false))))

(fn GetAbilityPlayerController
  (bind _returnvalue (Fortnite|GetFortPlayerControllerfromActor (.AsFort%20Player%20Pawn%20Athena (CallFunction|GetAbilityPlayerPawn))))
  (bind (_valid _asfort_player_pawn_athena) (CallFunction|GetAbilityPlayerPawn))
  (if _valid
    (Utilities|IsValid _returnvalue
      (:Is%20Valid
        (return true _returnvalue))
      (:Is%20Not%20Valid
        _))
    (else
      _))
  (return false 0))

(fn GetAbilityPlayerPawn
  (bind _asfort_player_pawn_athena (Utilities|Casting|CastToFortPlayerPawnAthena (Ability|GetAvatarActorfromActorInfo))
    (:then
      (return true _asfort_player_pawn_athena))
    (:CastFailed
      (return false 0))))
