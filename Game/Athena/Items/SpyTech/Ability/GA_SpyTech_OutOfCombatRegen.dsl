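; SpyTech gameplay ability that gives the player out-of-combat regeneration.
; On activation it stores the player pawn as target data and applies the regen gameplay effect to it.
; It then listens for damage dealt and damage received events in parallel.
; Either event applies the ignore-regen effect to the same target, presumably pausing regen during combat.
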(event Ability|EventActivateAbility
  (|CallParentFunction)
  (bind _playerpawn (Utilities|GetValidAbilityPlayerCharacter)
    (:Valid
      (bind _output_get (Variables|Default|SetTarget (Ability|TargetData|AbilityTargetDatafromActor _playerpawn)))
      (Ability|ApplyGameplayEffectToTarget _output_get (Variables|Default|GetGE_Regen))
      (Ability|Tasks|WaitGameplayEvent "(TagName=\"Event.Damage.Dealt\")"
        (:then
          (Ability|Tasks|WaitGameplayEvent "(TagName=\"Event.Damage.Received\")"
            (:EventReceived
              _)))
        (:EventReceived
          _))
      (Ability|ApplyGameplayEffectToTarget (Variables|Default|GetTarget) (Variables|Default|GetGE_IgnoreRegen)))))
