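; Base gameplay ability for SpyTech items, providing shared helpers to reach the owning player's pawn and controller.
; GetAbilityPlayerPawn casts the avatar actor to an Athena player pawn and returns it with a validity flag.
; GetAbilityPlayerController builds on that helper to get the pawn's Fort player controller, also with a validity flag.
; The activate, end and SpyTech granted events are empty in this class.
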(event Ability|EventActivateAbility)

(event Ability|EventOnEndAbility
  (params (bWasCancelled :type Boolean)))

(event SpyTech|EventOnAbilityGranted
  (params (TagContainer :type Gameplay%20Tag%20Container%20Structure%20%28by%20ref%29)))

(fn GetAbilityPlayerController
  (bind _returnvalue (Fortnite|GetFortPlayerControllerfromActor (.AsFort%20Player%20Pawn%20Athena (CallFunction|GetAbilityPlayerPawn))))
  (bind (_valid _asfort_player_pawn_athena) (CallFunction|GetAbilityPlayerPawn))
  (if _valid
    (Utilities|IsValid _returnvalue
      (:Is%20Valid
        (return true _returnvalue))
      (:Is%20Not%20Valid
        _))
    (else
      _))
  (return false 0))

(fn GetAbilityPlayerPawn
  (bind _asfort_player_pawn_athena (Utilities|Casting|CastToFortPlayerPawnAthena (Ability|GetAvatarActorfromActorInfo))
    (:then
      (return true _asfort_player_pawn_athena))
    (:CastFailed
      (return false 0))))
