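; SpyTech "Parting Gift" gameplay ability that, when triggered by a gameplay event, drops an explosive projectile above the player.
; It spawns the projectile over the pawn's capsule with no initial speed, attaching primary and additional explosion effect containers built from tags.
; If the projectile is a B_Prj_Athena_PurpleMouse, it prints a debug team message and assigns the player's team.
; The ability then ends, and also ends immediately without a valid player character.
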(event Ability|EventActivateAbilityFromEvent
  (params (EventData :type Gameplay%20Event%20Data%20Structure%20%28by%20ref%29))
  (bind _playerpawn (Variables|Default|GetPlayerPawn))
  (bind _returnvalue (Team|GetActorTeam _playerpawn))
  (|CallParentFunction EventData)
  (bind _playerpawn_2 (Utilities|GetValidAbilityPlayerCharacter)
    (:Valid
      (Variables|Default|SetPlayerPawn _playerpawn_2)
      (Variables|Default|SetAdditionalExplosionGECS (Ability|MakeGameplayEffectContainerSpec (Variables|Default|GetTag_AdditionalExplosionGameplayEffectContainer)))
      (bind _returnvalue_2 (Projectile|SpawnProjectile :ProjectileClass (Variables|Default|GetProjectile) :RequestedBy (Variables|Default|GetPlayerPawn) :SpawnLocation (Math|Vector|vector+vector (Transformation|GetActorLocation (Variables|Default|GetPlayerPawn)) (Math|Vector|MakeVector 0.0 0.0 (+ (Components|Capsule|GetScaledCapsuleHalfHeight (Variables|/Script/Engine.Character|GetCapsuleComponent (Variables|Default|GetPlayerPawn))) (Variables|Default|GetHeightAbovePawnToSpawn)))) :SpawnRotation (Variables|Default|GetSpawn%20Rotation) :SpawnDirection (Variables|Default|GetSpawn%20Direction) :EffectContainerSpecToApplyOnExplode (Ability|MakeGameplayEffectContainerSpec (Variables|Default|GetTag_ExplosionGameplayEffectContainer)) :bUseDefaultPhysics false :InitialSpeed 0.0 :GravityScale 1.0 :ChargePercent 0.0))
      (Delivery|AddAdditionalExplodeEffectContainer _returnvalue_2 (Variables|Default|GetAdditionalExplosionGECS))
      (bind _asb_prj_athena_purple_mouse (Utilities|Casting|CastToB_Prj_Athena_PurpleMouse _returnvalue_2)
        (:then
          (Development|PrintString (Utilities|String|Append "Ability fired with team " (Utilities|String|ToString(Byte) _returnvalue)) true true "(R=1.000000,G=0.002489,B=0.000000,A=1.000000)" 5.0)
          (Class|BPrjAthenaPurpleMouse|SetTeam _asb_prj_athena_purple_mouse _returnvalue :!class /Game/Athena/Items/Weapons/Prototype/PurpleMouse/B_Prj_Athena_PurpleMouse.B_Prj_Athena_PurpleMouse_C))
        (:CastFailed
          _))
      (Ability|EndAbility))
    (:Invalid
      (Ability|EndAbility))))
