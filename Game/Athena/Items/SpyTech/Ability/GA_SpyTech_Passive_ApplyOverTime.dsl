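; Passive SpyTech gameplay ability that applies an over-time Gameplay Effect to its owner.
; On activation it calls the parent, applies the class held in OverTimeEffectClass, then immediately ends the ability.
; The ability does not track or remove the effect afterward, and its OnEndAbility event is only a disabled stub.
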(event Ability|EventActivateAbility
  (|CallParentFunction)
  (Ability|ApplyGameplayEffectToOwner (Variables|SpyTech_Effect|GetOverTimeEffectClass))
  (Ability|EndAbility))

(event Ability|EventOnEndAbility
  (params (bWasCancelled :type Boolean))
  (|CallParentFunction bWasCancelled) ; This node is disabled and will not be called.\nDrag off pins to build functionality.
)
