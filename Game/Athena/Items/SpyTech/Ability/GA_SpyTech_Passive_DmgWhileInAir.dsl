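; SpyTech passive ability that grants a damage bonus to the player while airborne, as used with the Infiltration playlist data.
; On activation it reads the required air time from an Infiltration curve table and waits for in-air tags to be added or removed.
; An added tag starts a timer that rechecks the tags before applying the bonus effect.
; Removal waits a short debounce delay and rechecks before applying the removal effect.
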(event Ability|EventActivateAbility
  (|CallParentFunction)
  (bind _outxy (DataTable|EvaluateCurveTableRow "/Game/Athena/Playlists/Infiltration/OverrideGameData.OverrideGameData" "Infiltration.InAirTimeCheck")
    (:RowFound
      (Variables|Default|SetTime _outxy)
      (bind _playerpawn (Utilities|GetValidAbilityPlayerCharacter)
        (:Valid
          (Variables|Default|SetPlayer _playerpawn)
          (Variables|Default|SetTarget (Ability|TargetData|AbilityTargetDatafromActor _playerpawn))
          (Ability|Tasks|WaitGameplayTagAdd (Variables|Default|GetInAirTags) _playerpawn
            (:then
              _)
            (:Added
              (if (GameplayTags|HasAnyMatchingGameplayTags (Ability|GetAbilitySystemComponent (Variables|Default|GetPlayer)) (GameplayTags|MakeGameplayTagContainerfromArray (Variables|Default|GetInAirTags_Array)))
                (bind _returnvalue (Utilities|Time|SetTimerbyEvent (delegate Custom|Set%20Falling%20Bonus%20Dmg) (Variables|Default|GetTime)))
                (Variables|Default|SetFallingBonusDmgTimerHandle _returnvalue))))
          (Ability|Tasks|WaitGameplayTagRemove (Variables|Default|GetInAirTags)
            (:Removed
              (Ability|Tasks|WaitDelay 0.2
                (:OnFinish
                  (if (not (GameplayTags|HasAnyMatchingGameplayTags (Ability|GetAbilitySystemComponent (Variables|Default|GetPlayer)) (GameplayTags|MakeGameplayTagContainerfromArray (Variables|Default|GetInAirTags_Array))))
                    (Ability|ApplyGameplayEffectToTarget (Variables|Default|GetTarget) (Variables|Default|GetRemoveGe))
                    (Utilities|Time|ClearandInvalidateTimerbyHandle (Variables|Default|GetFallingBonusDmgTimerHandle))))))))))))

(event Custom|Set%20Falling%20Bonus%20Dmg
  (if (GameplayTags|HasAnyMatchingGameplayTags (Ability|GetAbilitySystemComponent (Variables|Default|GetPlayer)) (GameplayTags|MakeGameplayTagContainerfromArray (Variables|Default|GetInAirTags_Array)))
    (Ability|ApplyGameplayEffectToTarget (Variables|Default|GetTarget) (Variables|Default|GetAdd%20GE))
    (Utilities|Time|ClearandInvalidateTimerbyHandle (Variables|Default|GetFallingBonusDmgTimerHandle))))
