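; SpyTech passive gameplay ability that periodically converts the owning player's metal into shield.
; On activation, the dedicated server starts a looping timer at a scalable interval, which is cleared when the ability ends.
; Each tick skips players at max shield, checks inventory for enough metal, removes the scalable cost and applies the grant effect if removal succeeds.
; The timer event is named GrantShieldForAmmo, though it consumes metal.
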(event Ability|EventActivateAbility
  (|CallParentFunction)
  (if (Networking|IsDedicatedServer)
    (bind _returnvalue (Utilities|Time|SetTimerbyEvent (delegate Custom|GrantShieldForAmmo) (ScalableFloat|GetScalableFloatMultiplier (Variables|Default|GetGrantTime)) true))
    (Variables|Default|SetGrantTimeTimerHandle _returnvalue)))

(event Custom|GrantShieldForAmmo
  (if (not (CallFunction|IsatMaxShields))
    (bind _playercontroller (Utilities|GetValidAbilityPlayerController)
      (:Valid
        (bind _hasmetal (CallFunction|HasMetalinInventory _playercontroller))
        (if _hasmetal
          (bind _returnvalue (Inventory|RemoveItemFromPlayer :PlayerController _playercontroller :ItemDefinition (Variables|Default|GetMetalItemDefinition) :AmountToRemove (Math|Float|Truncate (ScalableFloat|GetScalableFloatMultiplier (Variables|Default|GetMetalCost))) :bForceRemoval true :bSkipVerbRemoval false))
          (if (> _returnvalue 0)
            (Ability|ApplyGameplayEffectToOwner (Variables|Default|GetGrantEffectClass))))))))

(event Ability|EventOnEndAbility
  (params (bWasCancelled :type Boolean))
  (|CallParentFunction bWasCancelled)
  (Utilities|Time|ClearandInvalidateTimerbyHandle (Variables|Default|GetGrantTimeTimerHandle)))

(fn HasMetalInInventory
  (params (PlayerController :type Fort%20Player%20Controller%20Object%20Reference))
  (bind _metalitemdefinition (Variables|Default|GetMetalItemDefinition))
  (bind _returnvalue (Inventory|FindExistingItemForDefinition :self PlayerController :ItemDefinition _metalitemdefinition :bInStorageVault false))
  (Utilities|IsValid _returnvalue
    (:Is%20Valid
      (return (>= (Item|GetNuminStack _returnvalue) (Math|Float|Truncate (ScalableFloat|GetScalableFloatMultiplier (Variables|Default|GetMetalCost)))) (Item|GetNuminStack _returnvalue)))
    (:Is%20Not%20Valid
      (return false 0))))

(fn IsAtMaxShields
  (bind _playerpawn (Utilities|GetValidAbilityPlayerCharacter)
    (:Valid
      (if (and (Ability|Attribute|GetFloatAttribute _playerpawn "(AttributeName=\"CurrentShield\",Attribute=/Script/FortniteGame.FortHealthSet:CurrentShield,AttributeOwner=None)") (Ability|Attribute|GetFloatAttribute _playerpawn "(AttributeName=\"Shield\",Attribute=/Script/FortniteGame.FortHealthSet:Shield,AttributeOwner=None)"))
        (return (>= (.ReturnValue (Ability|Attribute|GetFloatAttribute _playerpawn "(AttributeName=\"CurrentShield\",Attribute=/Script/FortniteGame.FortHealthSet:CurrentShield,AttributeOwner=None)")) (.ReturnValue (Ability|Attribute|GetFloatAttribute _playerpawn "(AttributeName=\"Shield\",Attribute=/Script/FortniteGame.FortHealthSet:Shield,AttributeOwner=None)"))))
        (else
          (return false))))
    (:Invalid
      (return false))))
