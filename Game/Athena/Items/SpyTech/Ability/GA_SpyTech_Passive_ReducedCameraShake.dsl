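; GA_SpyTech_Passive_ReducedCameraShake is a Spy Tech passive gameplay ability that reduces damage camera shake for its player.
; On activation it gets the valid player pawn and sets that pawn's DamageCameraShakeClass to the ReducedCameraShakeClass variable.
; It then calls End Ability.
; The end-ability event holds only a disabled parent call, so nothing in this graph restores the original shake class.
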(event Ability|EventActivateAbility
  (|CallParentFunction)
  (bind _playerpawn (Utilities|GetValidAbilityPlayerCharacter)
    (:Valid
      (Variables|/Script/FortniteGame.FortPlayerPawnAthena|SetDamageCameraShakeClass (Variables|Default|GetReducedCameraShakeClass) _playerpawn))
    (:Invalid
      _))
  (Ability|EndAbility))

(event Ability|EventOnEndAbility
  (params (bWasCancelled :type Boolean))
  (|CallParentFunction bWasCancelled) ; This node is disabled and will not be called.\nDrag off pins to build functionality.
)
