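; SpyTech passive gameplay ability that reloads a weapon when the player switches away from it.
; On activation it calls the parent, then binds the player pawn's weapon-equipped delegate.
; On each equip, it checks whether the previous weapon's data has the configured weapon tag and, if so, calls TryReloadAmmofromAbility on that weapon.
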(event Ability|EventActivateAbility
  (bind (_valid _asfort_player_pawn_athena) (CallFunction|GetAbilityPlayerPawn))
  (|CallParentFunction)
  (FortPawn|AssignOnWeaponEquippedDelegate _asfort_player_pawn_athena (delegate Custom|OnWeaponEquippedDelegate)))

(event Custom|OnWeaponEquippedDelegate
  (params (NewWeapon :type Fort%20Weapon%20Object%20Reference) (PrevWeapon :type Fort%20Weapon%20Object%20Reference))
  (if (GameplayTags|HasTag (GameplayTags|GetOwnedGameplayTags (Data|GetWeaponData PrevWeapon) :!class /Script/GameplayTags.GameplayTagAssetInterface) (Variables|Default|GetWeaponTag))
    (FortWeapon|TryReloadAmmofromAbility PrevWeapon)))
