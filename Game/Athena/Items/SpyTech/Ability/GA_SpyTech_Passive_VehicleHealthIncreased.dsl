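; GA_SpyTech_Passive_VehicleHealthIncreased is a passive SpyTech ability that applies a vehicle health effect while the player is in a vehicle.
; On activation it caches the player's ASC and starts gameplay tag add and remove waits on the in-vehicle tag.
; When the tag is added, the player's ASC applies the effect to the vehicle's ASC and stores the handle.
; When the tag is removed, one stack of that effect is removed.
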(event Ability|EventActivateAbility
  (|CallParentFunction)
  (bind _playerpawn (Utilities|GetValidAbilityPlayerCharacter)
    (:Valid
      (bind _output_get (Variables|Default|SetPlayer _playerpawn))
      (Variables|Default|SetPlayer%20ASC (Ability|GetAbilitySystemComponent _output_get))
      (Ability|Tasks|WaitGameplayTagAdd (Variables|Default|GetInVehicleTag)
        (:then
          (Ability|Tasks|WaitGameplayTagRemove (Variables|Default|GetInVehicleTag)
            (:Removed
              (GameplayEffects|RemoveActiveGameplayEffect (Variables|Default|GetVehicle%20ASC) (Variables|Default|GetGE%20Handle) 1))))
        (:Added
          (bind _output_get_2 (Variables|Default|SetVehicleActor (Vehicle|GetVehicleActor (Variables|Default|GetPlayer) :!class /Script/FortniteGame.FortPlayerPawn)))
          (Variables|Default|SetVehicle%20ASC (Ability|GetAbilitySystemComponent _output_get_2))
          (bind _returnvalue (GameplayEffects|ApplyGameplayEffectToTarget (Variables|Default|GetPlayer%20ASC) (Variables|Default|GetVehicle%20Health%20GE) (Variables|Default|GetVehicle%20ASC) 1.0 (GameplayEffects|MakeEffectContext (Variables|Default|GetVehicle%20ASC))))
          (Variables|Default|SetGE%20Handle _returnvalue))))))
