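; SpyTech Sharpened Stick gameplay ability that builds a stacking damage multiplier on its owner from pickaxe events against players.
; ShouldAbilityRespondToEvent accepts only player pawn targets whose instigator tags match the pickaxe tag container.
; A DoOnce-gated check, reached from server-side activation and from ability end, increments StreakCount and applies GE_DamageMult while below the scalable stack limit.
; Helper functions resolve the avatar's player pawn and controller.
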(event Ability|EventActivateAbilityFromEvent
  (params (EventData :type Gameplay%20Event%20Data%20Structure%20%28by%20ref%29))
  (|CallParentFunction EventData)
  (Utilities|IsServer
    (:true
      ;[1 1 1] Check if we've reached our stack limit
      (Utilities|FlowControl|DoOnce false
        (:Completed
          (if (< (Variables|Default|GetStreakCount) (Math|Float|Truncate (ScalableFloat|GetValueatLevel (Variables|Default|GetSF_StackLimit))))
            (Variables|Default|SetStreakCount (+ (Variables|Default|GetStreakCount) 1)) ; Increase streak
            (Ability|ApplyGameplayEffectToOwner (Variables|Default|GetGE_DamageMult))
            (Ability|EndAbility)
            (else
              _)))))
    (:false
      _))
  (Ability|EndAbility))

(event Ability|EventOnEndAbility
  (params (bWasCancelled :type Boolean))
  ;[1 1 1] Check if we've reached our stack limit
  (Utilities|FlowControl|DoOnce false
    (:Completed
      (if (< (Variables|Default|GetStreakCount) (Math|Float|Truncate (ScalableFloat|GetValueatLevel (Variables|Default|GetSF_StackLimit))))
        (Variables|Default|SetStreakCount (+ (Variables|Default|GetStreakCount) 1)) ; Increase streak
        (Ability|ApplyGameplayEffectToOwner (Variables|Default|GetGE_DamageMult))
        (Ability|EndAbility)
        (else
          (Ability|EndAbility))))))

(fn GetAbilityPlayerController
  (bind _returnvalue (Fortnite|GetFortPlayerControllerfromActor (.AsFort%20Player%20Pawn%20Athena (CallFunction|GetAbilityPlayerPawn))))
  (bind (_valid _asfort_player_pawn_athena) (CallFunction|GetAbilityPlayerPawn))
  (if _valid
    (Utilities|IsValid _returnvalue
      (:Is%20Valid
        (return true _returnvalue))
      (:Is%20Not%20Valid
        _))
    (else
      _))
  (return false 0))

(fn GetAbilityPlayerPawn
  (bind _asfort_player_pawn_athena (Utilities|Casting|CastToFortPlayerPawnAthena (Ability|GetAvatarActorfromActorInfo))
    (:then
      (return true _asfort_player_pawn_athena))
    (:CastFailed
      (return false 0))))

(fn ShouldAbilityRespondToEvent
  (params (ActorInfo :type Gameplay%20Ability%20Actor%20Info%20Structure) (Payload :type Gameplay%20Event%20Data%20Structure))
  (bind (_eventtag _instigator _target _optionalobject _optionalobject2 _contexthandle _instigatortags) (Utilities|Struct|BreakGameplayEventData Payload))
  (Utilities|Casting|CastToFortPlayerPawn _target
    (:then
      (return (GameplayTags|TagContainer|CompareTagContainertoOtherTagContainers (Variables|Default|GetTC_Pickaxe) _instigatortags)))
    (:CastFailed
      (return false))) ; Only trigger off players
)
