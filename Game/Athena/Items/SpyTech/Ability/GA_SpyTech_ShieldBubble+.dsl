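; Event-triggered SpyTech gameplay ability that adds an area-of-effect cloud to a SilverBlazer shield bubble.
; On activation it takes the shield core from the event's OptionalObject and polls on a looping timer until the core's spawned bubble is valid.
; It then spawns the cloud with a gameplay effect container spec, attaches it to the core, passes it both shield references and ends.
; A safety timer ends the ability if no bubble appears.
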(event Ability|EventActivateAbilityFromEvent
  (params (EventData :type Gameplay%20Event%20Data%20Structure%20%28by%20ref%29))
  (bind (_eventtag _instigator _target _optionalobject) (Utilities|Struct|BreakGameplayEventData EventData))
  ;[1 1 1] Activate Ability from Event
  (|CallParentFunction EventData)
  (bind _asbga_athena_silver_blazer_core (Utilities|Casting|CastToBGA_Athena_SilverBlazerCore _optionalobject)
    (:then
      (Variables|Default|SetShield%20Core _asbga_athena_silver_blazer_core)
      (bind _returnvalue (Utilities|Time|SetTimerbyEvent (EventDispatchers|CreateEvent :!event Custom|WaitShieldSpawn) (Variables|Default|GetWait%20Time%20Before%20Spawn%20Check) true))
      (Variables|TimerHandles|SetTimer_WaitShieldSpawn _returnvalue)
      (bind _returnvalue_2 (Utilities|Time|SetTimerbyEvent (EventDispatchers|CreateEvent :!event Custom|Clear%20Timers%20and%20End%20Ability) (ScalableFloat|GetValueatLevel (Variables|Default|GetSafetyTimeOutDuration))))
      (Variables|TimerHandles|SetTimer_EndAbility _returnvalue_2))
    (:CastFailed
      (Ability|EndAbility))))

(event Custom|WaitShieldSpawn
  (bind _shield_core (Variables|Default|GetShield%20Core))
  (bind _spawnedbga (Variables|/Game/Athena/Items/Consumables/SilverBlazer/BGA_Athena_SilverBlazerCore.BGA_Athena_SilverBlazerCore_C|GetSpawnedBGA _shield_core))
  ;[1 1 1] Wait Shield Spawn, then spawn cloud actor, attach it to shield, and pass through actor references
  (Utilities|IsValid _spawnedbga
    (:Is%20Valid
      (Variables|Default|SetShieldBubble _spawnedbga)
      (bind _returnvalue (Ability|MakeGameplayEffectContainerSpecfromAbility (Variables|Default|GetGE_Container) self 0))
      (Variables|Default|SetGE_ContainerSpec _returnvalue)
      (Utilities|Time|ClearandInvalidateTimerbyHandle (Variables|TimerHandles|GetTimer_WaitShieldSpawn))
      (bind _returnvalue_2 (AreaOfEffect|SpawnAreaOfEffectCloud (Variables|Default|GetAoE_ActorClass) (Ability|GetAvatarActorfromActorInfo) (Transformation|GetActorLocation (Variables|Default|GetShield%20Core)) (0.0 0.0 0.0) (Variables|Default|GetGE_ContainerSpec) (Variables|Default|GetGE_ContainerSpec)))
      (bind _output_get (Variables|Default|SetAoE%20Actor _returnvalue_2))
      (Transformation|AttachActorToActor _output_get (Variables|Default|GetShield%20Core) "None" "SnapToTarget" "KeepRelative" "KeepWorld" true)
      (bind _asb_spy_tech_shield_bubble__ao_e (Utilities|Casting|CastToB_SpyTech_ShieldBubble+_AoE (Variables|Default|GetAoE%20Actor)))
      (Class|BSpyTechShieldBubble+AoE|UpdateShieldRefs _asb_spy_tech_shield_bubble__ao_e (Variables|Default|GetShield%20Core) (Variables|Default|GetShieldBubble))
      (CallFunction|ClearTimersAndEndAbility))))

(event Custom|Clear%20Timers%20and%20End%20Ability
  ;[1 1 1] Clear Timers and End Ability
  (Utilities|Time|ClearandInvalidateTimerbyHandle (Variables|TimerHandles|GetTimer_WaitShieldSpawn))
  (Utilities|Time|ClearandInvalidateTimerbyHandle (Variables|TimerHandles|GetTimer_EndAbility))
  (Ability|EndAbility))
