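; GA_SpyTech_ShieldSteal is an event-triggered gameplay ability for the SpyTech shield steal, granting the owner shield based on an incoming event's magnitude.
; On activation it multiplies the event magnitude by the ShieldStealPercent scalable float.
; The result becomes the SetByCaller.Healing magnitude on a GE_ShieldHeal spec, which is applied to the owner before the ability ends.
; The event's target is not used, so this graph only heals and removes no shield from another actor.
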(event Ability|EventActivateAbilityFromEvent
  (params (EventData :type Gameplay%20Event%20Data%20Structure%20%28by%20ref%29))
  (bind (_eventtag _instigator _target _optionalobject _optionalobject2 _contexthandle _instigatortags _targettags _eventmagnitude) (Utilities|Struct|BreakGameplayEventData EventData))
  (|CallParentFunction EventData)
  (bind _returnvalue (Ability|GameplayEffect|AssignTagSetbyCallerMagnitude (Ability|MakeOutgoingGameplayEffectSpec (Variables|Default|GetGE_ShieldHeal)) "(TagName=\"SetByCaller.Healing\")" (* _eventmagnitude (ScalableFloat|GetValueatLevel (Variables|Default|GetShieldStealPercent)))))
  (Ability|ApplyGameplayEffectSpecToOwner _returnvalue)
  (Ability|EndAbility))
