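; GA_SpyTech_ShieldSynergy is a SpyTech gameplay ability that grants the player shield whenever their health increases.
; On activation it caches the player pawn's current health and binds to the pawn's health-changed event.
; Each change computes the delta from the cached health and re-caches it.
; Positive deltas, scaled by a ScalableFloat percentage, go via a SetByCaller healing magnitude into a shield-heal gameplay effect applied to the owner.
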(event Ability|EventActivateAbility
  ;[1 1 1] On activate ability, save health and bind to health changed
  (|CallParentFunction)
  (bind _playerpawn (Utilities|GetValidAbilityPlayerCharacter)
    (:Valid
      (Variables|Default|SetPlayer _playerpawn)
      (Variables|Default|SetPreviousHealth (FortPawn|GetHealth (Variables|Default|GetPlayer)))
      (Health|BindEventtoOnPawnHealthChanged (Variables|Default|GetPlayer) (EventDispatchers|CreateEvent :!event Custom|HealthChanged)))))

(event Custom|HealthChanged
  ;[1 1 1] On Health changed, calc delta health and resave previous health. If delta health is positive, heal.
  (Variables|Default|SetDeltaHealth (- (FortPawn|GetHealth (Variables|Default|GetPlayer)) (Variables|Default|GetPreviousHealth)))
  (Variables|Default|SetPreviousHealth (FortPawn|GetHealth (Variables|Default|GetPlayer)))
  (if (> (Variables|Default|GetDeltaHealth) 0.0)
    (bind _returnvalue (Ability|GameplayEffect|AssignTagSetbyCallerMagnitude (Ability|MakeOutgoingGameplayEffectSpec (Variables|Default|GetGE_ShieldHeal)) "(TagName=\"SetByCaller.Healing\")" (* (Variables|Default|GetDeltaHealth) (ScalableFloat|GetValueatLevel (Variables|Default|GetPercentHealingReceivedAsShields)))))
    (Ability|ApplyGameplayEffectSpecToOwner _returnvalue)))
