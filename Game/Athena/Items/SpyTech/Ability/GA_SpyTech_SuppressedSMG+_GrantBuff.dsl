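; Gameplay ability for the SpyTech Suppressed SMG+ that grants its owner a buff while the weapon's magazine is low.
; On a gameplay event it stores the weapon, taken from the payload for weapon switches or from the pawn's current weapon otherwise.
; It compares magazine ammo to a threshold from per-shot cost and a scalable shot count, applies or removes the buff effect, prints debug strings, then ends.
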(event Ability|EventActivateAbilityFromEvent
  (params (EventData :type Gameplay%20Event%20Data%20Structure%20%28by%20ref%29))
  (bind (_eventtag _instigator _target _optionalobject) (Utilities|Struct|BreakGameplayEventData EventData))
  (bind _weapon (Variables|Default|GetWeapon))
  (bind _returnvalue (FortWeapon|GetMagazineAmmoCount _weapon))
  (|CallParentFunction EventData)
  (switch Utilities|FlowControl|Switch|SwitchonGameplayTag _eventtag
    (:Default
      (bind _playerpawn (Utilities|GetValidAbilityPlayerCharacter)
        (:Valid
          (Variables|Default|SetWeapon (Variables|/Script/FortniteGame.FortPawn|GetCurrentWeapon _playerpawn)))
        (:Invalid
          (Ability|EndAbility))))
    (:Event.Weapons.WeaponSwitch
      (bind _asfort_weapon (Utilities|Casting|CastToFortWeapon _optionalobject)
        (:then
          (Variables|Default|SetWeapon _asfort_weapon))
        (:CastFailed
          (Ability|EndAbility)))))
  (if (<= _returnvalue (* (FortWeapon|GetAmmoCostPerFire _weapon 0.0) (- (Math|Float|Round (ScalableFloat|GetValueatLevel (Variables|Default|GetNumberOfShotsToBuff))) 1)))
    (Ability|ApplyGameplayEffectToOwner (Variables|Default|GetGE_Buff) 0)
    (Development|PrintString "Add Buff")
    (else
      (Ability|RemoveGameplayEffectFromOwnerWithGrantedTags (Variables|Default|GetTag_BuffGranted))
      (Development|PrintString "Remove Buff")))
  (Development|PrintString (Utilities|String|ToString(Integer) _returnvalue))
  (Ability|EndAbility))
