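; SpyTech passive ability that makes the player dance after earning a player elimination or killing a tag-matched NPC.
; TryBeginDance applies the dance effect and optional heal effect unless ignore tags are present.
; Tag add/remove tasks count hiding-prop and delayed-action states, deferring the dance until exit plus a delay.
; When hidden, it can send StopHiding events for hiding props or sneaky snowmen to force an exit.
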(event Ability|EventFailedToActivatePassiveAbility
  (params (ActorInfo :type Gameplay%20Ability%20Actor%20Info%20Structure))
  (|CallParentFunction))

(event Ability|EventActivateAbility
  ;[1 1 1] Activate Ability
  (|CallParentFunction)
  (bind _playerpawn (Utilities|GetValidAbilityPlayerCharacter)
    (:Valid
      (Variables|HelperVars|SetPlayerPawn _playerpawn)
      (CallFunction|HandleHidingProps)
      (CallFunction|HandleDelayedActions)
      (Ability|Tasks|WaitGameplayEvent (Variables|Default|Events|GetEvent_EarnedElimination(Player)) (Variables|HelperVars|GetPlayerPawn)
        (:then
          (bind (_asynctaskproxy _payload) (Ability|Tasks|WaitGameplayEvent (Variables|Default|Events|GetEvent_Killed(NPC)) (Variables|HelperVars|GetPlayerPawn))
            (:EventReceived
              (if (GameplayTags|HasAllTags (.TargetTags (Utilities|Struct|BreakGameplayEventData _payload)) (Variables|Default|Tags|GetRequiredNPCTags))
                (CallFunction|TryBeginDance)))))
        (:EventReceived
          (CallFunction|TryBeginDance))))
    (:Invalid
      (Ability|EndAbility))))

(event Custom|HandleHidingProps
  ;[1 1 1] Handle Hidable Props
  (bind (_asynctaskproxy _payload) (Ability|Tasks|WaitGameplayEvent (Variables|Default|Events|GetEvent_UpdatePropReference) (Variables|HelperVars|GetPlayerPawn))
    (:then
      (for _array_element (GameplayTags|BreakGameplayTagContainer (Variables|Default|Tags|GetTags_HidableProps))
        (Ability|Tasks|WaitGameplayTagAdd _array_element
          (:then
            (Ability|Tasks|WaitGameplayTagRemove _array_element
              (:Removed
                (Math|Integer|DecrementInt (Variables|HelperVars|GetHidablePropTagCount)
                  (:%20%20
                    (if (<= (Variables|HelperVars|GetHidablePropTagCount) 0)
                      (Variables|HelperVars|SetHidablePropTagCount 0)
                      (if (Variables|HelperVars|GetTryDanceAfterExitingHidableProp)
                        (Variables|HelperVars|SetTryDanceAfterExitingHidableProp false)
                        (Utilities|FlowControl|Delay (ScalableFloat|GetValueatLevel (Variables|Default|Delays|GetDelay_HidablePropExit)))
                        (CallFunction|TryBeginDance))))))))
          (:Added
            (Math|Integer|IncrementInt (Variables|HelperVars|GetHidablePropTagCount))))))
    (:EventReceived
      (bind _asactor (Utilities|Casting|CastToActor (.OptionalObject (Utilities|Struct|BreakGameplayEventData _payload))))
      (Variables|HelperVars|SetHidingProp _asactor))))

(event Custom|HandleDelayedActions
  ;[1 1 1] Handle Delayed Actions
  (for _array_element (GameplayTags|BreakGameplayTagContainer (Variables|Default|Tags|GetTags_DelayedActions))
    (Ability|Tasks|WaitGameplayTagAdd _array_element
      (:then
        (Ability|Tasks|WaitGameplayTagRemove _array_element
          (:Removed
            (Math|Integer|DecrementInt (Variables|HelperVars|GetDelayedActionTagCount)
              (:%20%20
                (if (<= (Variables|HelperVars|GetDelayedActionTagCount) 0)
                  (Variables|HelperVars|SetDelayedActionTagCount 0)
                  (if (Variables|HelperVars|GetTryDanceAfterFinishingDelayedActions)
                    (Variables|HelperVars|SetTryDanceAfterFinishingDelayedActions false)
                    (Utilities|FlowControl|Delay (ScalableFloat|GetValueatLevel (Variables|Default|Delays|GetDelay_DelayedActionExit)))
                    (CallFunction|TryBeginDance))))))))
      (:Added
        (Math|Integer|IncrementInt (Variables|HelperVars|GetDelayedActionTagCount))))))

(event Custom|TryBeginDance
  ;[1 1 1] Try Begin Dance
  (Utilities|IsValid (Variables|HelperVars|GetPlayerPawn)
    (:Is%20Valid
      (if (<= (Variables|HelperVars|GetHidablePropTagCount) 0)
        (if (<= (Variables|HelperVars|GetDelayedActionTagCount) 0)
          (if (not (GameplayTags|HasAnyMatchingGameplayTags (Variables|HelperVars|GetPlayerPawn) (Variables|Default|Tags|GetTags_IgnoreDanceEffect)))
            (Ability|ApplyGameplayEffectToOwner (Variables|Default|GetGE_Dance) 0)
            (if (> (ScalableFloat|GetValueatLevel (Variables|Default|GetB_HealEnabled)) 0.0)
              (Ability|ApplyGameplayEffectToOwner (Variables|Default|GetGE_Heal) 0)))
          (else
            (Variables|HelperVars|SetTryDanceAfterFinishingDelayedActions true)))
        (else
          (Variables|HelperVars|SetTryDanceAfterExitingHidableProp true)
          (if (GameplayTags|HasAnyMatchingGameplayTags (Variables|HelperVars|GetPlayerPawn) (GameplayTags|MakeGameplayTagContainerfromArray (Utilities|Array|MakeArray (Variables|Default|Tags|GetTag_TeleportingInProp))))
            _
            (elif (GameplayTags|HasMatchingGameplayTag (Variables|HelperVars|GetPlayerPawn) (Variables|Default|Tags|GetTag_GettingInProp))
              (Utilities|FlowControl|Delay (ScalableFloat|GetValueatLevel (Variables|Default|Delays|GetDelay_GettingInProp)))
              (Utilities|IsValid (Variables|HelperVars|GetHidingProp)
                (:Is%20Valid
                  (Utilities|Casting|CastToB_HidingProp (Variables|HelperVars|GetHidingProp)
                    (:then
                      (Ability|SendGameplayEventtoActor (Variables|HelperVars|GetPlayerPawn) "(TagName=\"Event.Athena.HidingProp.StopHiding\")" (Utilities|Struct|MakeGameplayEventData :Instigator (Variables|HelperVars|GetHidingProp) :OptionalObject (Variables|HelperVars|GetHidingProp) :ContextHandle "()" :InstigatorTags "(GameplayTags=)" :TargetTags "(GameplayTags=)" :EventMagnitude 1.0 :TargetData "()")))
                    (:CastFailed
                      (Ability|SendGameplayEventtoActor (Variables|HelperVars|GetPlayerPawn) "(TagName=\"Event.Athena.SneakySnowman.StopHiding\")")))))
              (else
                _))))))
    (:Is%20Not%20Valid
      (Ability|EndAbility))))
